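; Thrown lightsaber ("Lobster" Realm) projectile that homes on enemies, damages and knocks them back, then returns to its owner.
; Timelines drive speed and spin until an apex timer or blocking hit triggers the return.
; Server-side overlap and stop events apply gameplay-effect damage, with temporary per-character immunity and pass-through for obstacles hit while returning.
; The weapon stays holstered until cleanup; killing the projectile sends the catch event and delays destruction for replication.
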(event EventEndPlay
  (params (EndPlayReason :type EEndPlayReason%20Enum :default "Destroyed"))
  (bind _physical_wash_force (Variables|Default|GetPhysical%20Wash%20Force))
  (|CallParentFunction EndPlayReason)
  ;[1 1 1] Stop audio if gets stuck (fixing issue with bug that audio gets stuck if throwing lobster right at the ground)
  (bind _castinairaudio (Variables|Default|GetCastInAirAudio)
    (:then
      (Audio|Components|Audio|FadeOut _castinairaudio 0.5 0.0 "Linear" :!class /Script/Engine.AudioComponent))
    (:else
      _))
  (bind _recallaudio (Variables|Default|GetRecallAudio)
    (:then
      (Audio|Components|Audio|FadeOut _recallaudio 0.5 0.0 "Linear" :!class /Script/Engine.AudioComponent)
      (Utilities|IsValid _physical_wash_force
        (:Is%20Valid
          (Actor|DestroyActor _physical_wash_force))) ; Checks if Rotor Wash Actor is Valid
    )
    (:else)))

(event Custom|CheckHomingTarget
  ;[1 1 1] Set Homing variables
  (Networking|SwitchHasAuthority
    (:Authority
      (bind _homingtarget (Variables|Default|GetHomingTarget)
        (:then
          (if (and (Utilities|IsValid _homingtarget :!class /Script/Engine.KismetSystemLibrary) (ScalableFloat|GetValueasBool (Variables|BPSetVariables|GetRow_HomingEnabled)))
            (bind _targetoffset (CallFunction|ChooseHomingTargetOffset _homingtarget))
            ;;[1 0 0.01573] Delete var once we can access homing data from projectile movement comp (only setter expoed to bp)
            (bind _output_get (Variables|Default|SetProjectile%20Homing%20Data (Utilities|Struct|MakeProjectileHomingData "LockOn" (ScalableFloat|GetValueatLevel (Variables|BPSetVariables|GetRow_HomingMinTurnSpeed)) (ScalableFloat|GetValueatLevel (Variables|BPSetVariables|GetRow_HomingMaxTurnSpeed)) (ScalableFloat|GetValueatLevel (Variables|BPSetVariables|GetRow_HomingTurnRampTime)) (ScalableFloat|GetValueatLevel (Variables|BPSetVariables|GetRow_HomingEffectTriggerRange)) 13 _homingtarget 0 "0.000000,0.000000,0.000000" _targetoffset "Auto" true true)))
            (Projectile|SetHomingData (Variables|Movement|GetProjectileMovementComponent) _output_get :!class /Script/FortniteGame.FortProjectileMovementComponent)
            (Utilities|Time|ClearandInvalidateTimerbyHandle (Variables|Default|GetCheckHomingTimer))
            (else
              _)))
        (:else
          (if (>= (- (Utilities|Time|GetGameTimeinSeconds) (Variables|Default|GetHomingCheckStartTime)) 0.1)
            _)))
      (Utilities|Time|ClearandInvalidateTimerbyHandle (Variables|Default|GetCheckHomingTimer)))))

(event Custom|ScaleMovementSpeed
  ;[1 1 1] Scale the movement speed after it is thrown
  ;;[1 1 1] Just Get the values once since this is called every frame for a second.
  (Variables|BPSetVariables|SetInitalScalingSpeed (ScalableFloat|GetValueatLevel (Variables|BPSetVariables|GetRow_StartVelocity)))
  (Variables|BPSetVariables|SetSlowScalingSpeed (ScalableFloat|GetValueatLevel (Variables|BPSetVariables|GetRow_EndVelocity)))
  (Components|Timeline|SetPlayRate (Variables|B_Prj_Realm_ThrownLobster|GetMovementSpeedScale) (/ 1.0 (ScalableFloat|GetValueatLevel (Variables|BPSetVariables|GetRow_AutoReturnTime))))
  (bind (_direction _speedratio) (Timeline|MovementSpeedScale PlayFromStart)
    (:Update
      (Variables|/Script/Engine.ProjectileMovementComponent|SetMaxSpeed (Math|Float|Lerp (Variables|BPSetVariables|GetInitalScalingSpeed) (Variables|BPSetVariables|GetSlowScalingSpeed) _speedratio) (Variables|Movement|GetProjectileMovementComponent)))))

(event EventBeginPlay
  (bind _skelmesh_weapon (Variables|Default|GetSkelMesh_Weapon))
  (bind _playerpawn (Variables|BPSetVariables|GetPlayerPawn))
  (bind _projectilemovementcomponent (Variables|Movement|GetProjectileMovementComponent))
  (bind _velocity (Variables|/Script/Engine.MovementComponent|GetVelocity _projectilemovementcomponent))
  (bind (_relativerotation_roll _relativerotation_pitch) (Math|Rotator|BreakRotator (Variables|/Script/Engine.SceneComponent|GetRelativeRotation_Roll (Variables|Default|GetSkelMesh_Weapon))))
  ;[1 1 1] EVENT -> Begin Play
  (Utilities|FlowControl|Sequence
    (:then_0
      (bind _returnvalue (Audio|SpawnSoundAttached (Variables|Default|GetCastSound) (Variables|Collision|GetCapsuleComponent)))
      (Utilities|IsValid _playerpawn
        (:Is%20Valid
          (Audio|Parameter|SetBooleanParameter _returnvalue "IsLocal" (Pawn|IsLocallyViewed _playerpawn) :!class /Script/AudioExtensions.AudioParameterControllerInterface)
          (Variables|Default|SetCastInAirAudio _returnvalue)
          (bind _castinairaudio (Variables|Default|GetCastInAirAudio)
            (:then
              (SoundIndicator|StartSoundIndicatorLooping (Variables|Default|GetFortSoundIndicator) (Transformation|GetActorLocation) (Variables|/Script/Engine.AudioComponent|GetSound _castinairaudio) 1.0 0))
            (:else)))))
    (:then_1
      ;[1 1 1] Set Homing variables
      (bind _returnvalue_2 (Utilities|Time|SetTimerbyEvent (delegate Custom|CheckHomingTarget) 0.001 true))
      (Variables|Default|SetCheckHomingTimer _returnvalue_2)
      (Variables|Default|SetHomingCheckStartTime (Utilities|Time|GetGameTimeinSeconds)))
    (:then_2
      (Game|Components|ProjectileMovement|Interpolation|SetInterpolatedComponent (Variables|Movement|GetProjectileMovementComponent) (Variables|Default|GetScene)) ; Interpolate the scene component so this doesn't affect the rotation of the saber
    )
    (:then_3
      (Utilities|Map|Clear (Variables|Default|GetRecentlyHit))
      (Variables|BPSetVariables|SetPreviousVelocity _velocity)
      (Variables|BPSetVariables|SetReturnVelocity (Math|Vector|NegateVector _velocity))
      (CallFunction|ScaleMovementSpeed)
      (Variables|BPSetVariables|SetLastHitNonDestroyedBuilding 0) ; Set LastHitNonDestroyed Building to null
      (bind _asfort_player_pawn (Utilities|Casting|CastToFortPlayerPawn (Game|GetInstigator)))
      (Variables|BPSetVariables|SetPlayerPawn _asfort_player_pawn)
      (bind _returnvalue_3 (Utilities|Time|SetTimerbyEvent (EventDispatchers|CreateEvent :!event Custom|NaturalApexReached) (ScalableFloat|GetValueatLevel (Variables|BPSetVariables|GetRow_AutoReturnTime))))
      (Variables|Default|SetNaturalApexReachedTimerHandle _returnvalue_3)
      (bind _returnvalue_4 (Utilities|Time|SetTimerbyEvent (EventDispatchers|CreateEvent :!event Custom|KillProjectile) (ScalableFloat|GetValueatLevel (Variables|BPSetVariables|GetRow_FailsafeMaxLifeTime)))) ; In case something goes wrong, ensure ability ends
      (Variables|BPSetVariables|SetFailsafeTimerHandle _returnvalue_4)
      ;[1 1 1] If player dies, destroy projectile
      (EventDispatchers|BindEventtoOnDied (Variables|BPSetVariables|GetPlayerPawn) (delegate Custom|OnDied))
      (Game|BindEventtoOnDestroyed (EventDispatchers|CreateEvent :!event Custom|ProjectileDestroyed))
      ;[1 1 1] Rotation in air
      (Components|Movement|SetUpdatedComponent (Variables|Default|GetRotatingMovement) (Variables|Default|GetSkelMesh_Weapon))
      (Components|Timeline|SetPlayRate (Variables|B_Prj_Realm_ThrownLobster|GetRotationSlowdown) (/ 1.0 (ScalableFloat|GetValueatLevel (Variables|BPSetVariables|GetRow_RotationSlowdownTime))))
      (bind (_direction _scalerotation _flattenrot) (Timeline|RotationSlowdown PlayFromStart)
        (:Update
          (Variables|/Script/Engine.RotatingMovementComponent|SetRotationRate_Roll (Math|Rotator|MakeRotator 0.0 0.0 (Math|Float|Lerp (ScalableFloat|GetValueatLevel (Variables|BPSetVariables|GetRow_StartSpinRate)) (ScalableFloat|GetValueatLevel (Variables|BPSetVariables|GetRow_EndSpinRate)) _scalerotation)) (Variables|Default|GetRotatingMovement)) ; decelerate the rotation before it stops moving
          (if (Variables|Default|GetInitialRotationFromHandOrientation)
            (Transformation|SetActorRelativeRotation (Math|Rotator|CombineRotators (Utilities|Operators|Multiply (Math|Rotator|MakeRotator (- _relativerotation_pitch 9e+01)) _flattenrot) (Transformation|GetActorRotation))) ; slowly roll the entire saber over from its initial orientation to a horizontal orientation
          ))) ; Scale up rotation rate so it isn't rotating instantly when it comes out of hand
    )
    (:then_4
      (bind _asfort_player_pawn_2 (Utilities|Casting|CastToFortPlayerPawn (Actor|GetOwner)))
      (bind _output_get (Variables|Default|SetProjectileOwnerPawn _asfort_player_pawn_2))
      ;;[0.4577 0.7777 1] Cache the realm weapon so we can access it later even if it is unequipped
      (bind _asb_athena_galileo_lobster_realm (Utilities|Casting|CastToB_Athena_Galileo_Lobster_Realm (Weapon|GetCurrentWeapon _output_get))
        (:then
          (Variables|Default|SetOwningRealmWeapon _asb_athena_galileo_lobster_realm))
        (:CastFailed
          _))
      (Utilities|IsValid _skelmesh_weapon
        (:Is%20Valid
          (if (Variables|Default|GetInitialRotationFromHandOrientation)
            (Transformation|SetWorldRotation _skelmesh_weapon (Transformation|GetSocketRotation (Variables|/Script/Engine.Character|GetMesh (Variables|Default|GetProjectileOwnerPawn)) "hand_r") false false)))))
    (:then_5
      (Networking|SwitchHasAuthority
        (:Authority
          (bind _owningrealmweapon (Variables|Default|GetOwningRealmWeapon)
            (:then
              (Class|BAthenaGalileoLobsterRealm|SetProjectile _owningrealmweapon self))
            (:else)))))
    (:then_6
      ;;[0.453 1 0.9568] Since the weapon should be holstered while the projectile exists, have the projectile be the holster source and holster the weapon upon its creation
      (bind _playerpawn_2 (Variables|BPSetVariables|GetPlayerPawn)
        (:then
          (Weapon|SetHolsterWeapon _playerpawn_2 0 true false "realm"))
        (:else)))
    (:then_7)))

(event Custom|OnDied
  (params (DamagedActor :type Actor%20Object%20Reference) (Damage :type Float%20%28single-precision%29) (InstigatedBy :type Controller%20Object%20Reference) (DamageCauser :type Actor%20Object%20Reference) (HitLocation :type Vector) (FHitComponent :type Primitive%20Component%20Object%20Reference) (BoneName :type Name) (Momentum :type Vector))
  ;[1 1 1] If player dies, destroy projectile
  (CallFunction|KillProjectile))

(event OnProjectileStop(ProjectileMovementComponent)
  (params (ImpactResult :type Hit%20Result%20Structure%20%28by%20ref%29))
  (bind _returnvalue (Utilities|GetDisplayName (.HitActor (Collision|BreakHitResult ImpactResult))))
  (bind (_bblockinghit _binitialoverlap _time _distance _location _impactpoint _normal _impactnormal _physmat _hitactor) (Collision|BreakHitResult ImpactResult))
  (bind _lasthitobstacle (Variables|BPSetVariables|GetLastHitObstacle))
  ;[0.6406 1 0.6843] EVENT -> On Projectile Stop
  (Networking|SwitchHasAuthority
    (:Authority
      (Utilities|IsValid _hitactor
        (:Is%20Valid
          (if (Variables|/Script/Engine.Actor|GetbCanBeDamaged _hitactor)
            (Variables|Default|SetHitResult ImpactResult)
            (bind _asfort_vehicle_interface (Utilities|Casting|CastToFortVehicleInterface _hitactor)
              (:then
                ;;[0.1752 1 0.4401] Hit and Damage Vehicles
                (CallFunction|DamageHitActor (Vehicle|GetActor _asfort_vehicle_interface))
                (Development|PrintString "hit vehicle" true true "(R=0.000000,G=1.000000,B=0.020782,A=1.000000)" 5.0)
                (Collision|IgnoreActorwhenMoving (Variables|Collision|GetCapsuleComponent) (Vehicle|GetActor _asfort_vehicle_interface) true)
                (if (Variables|BPSetVariables|GetReturnToSender)
                  (CallFunction|HandleProjectileStopReturn)
                  (else
                    (CallFunction|ReturntoOwner))))
              (:CastFailed
                ;;[1 0.4908 0.4609] Hit and Damaged Building
                (bind _asbuilding_actor (Utilities|Casting|CastToBuildingActor _hitactor)
                  (:then
                    (bind _output_get (Variables|BPSetVariables|SetLastHitObstacle _asbuilding_actor))
                    (if (and (and (> (Health|GetMaxHealth _output_get :!class /Script/FortniteGame.BuildingActor) 0.0) (<= (Health|GetMaxHealth _output_get :!class /Script/FortniteGame.BuildingActor) 5e+03)) (!= _output_get (Variables|BPSetVariables|GetLastHitNonDestroyedBuilding)))
                      (CallFunction|DamageEnvironmentActor _lasthitobstacle)
                      (if (<= (Health|GetHealth _lasthitobstacle :!class /Script/FortniteGame.BuildingActor) 0)
                        (Development|PrintString (Utilities|String|Append "BROKE AND CONTINUE:   " (Utilities|GetDisplayName (Variables|BPSetVariables|GetLastHitObstacle))))
                        (if (Variables|BPSetVariables|GetReturnToSender)
                          (CallFunction|HandleProjectileStopReturn) ; continue returning to player
                          (else
                            (CallFunction|HandleProjectileStop) ; continue moving forward
                          ))
                        (else
                          ;;[1 0.4524 0.2474] Hit a destructible item, but did not destroy it so return to player
                          (bind _output_get_2 (Variables|BPSetVariables|SetLastHitNonDestroyedBuilding (Variables|BPSetVariables|GetLastHitObstacle))) ; Save this building so we can check it in ViableHit?\nThis fixes buildings taking multiple hits when the saber is "bouncing" off\nof them or passing through them. 
                          (Development|PrintString (Utilities|String|Append "Damaged and STOP:   " (Utilities|GetDisplayName _output_get_2)))))
                      (else
                        (Development|PrintString (Utilities|String|Append "ViableHit returned false, ignoring hit on " (Utilities|GetDisplayName (Variables|BPSetVariables|GetLastHitObstacle))))))
                    (Transformation|SetActorLocation (+ (Transformation|GetActorLocation) (* (Math|Vector|GetUnitDirection(Vector) (Transformation|GetActorLocation) (Transformation|GetActorLocation (Variables|BPSetVariables|GetPlayerPawn))) 1e+02)) false true) ; If we hit something solid then we need to move projectile \ntowards player or it will be stuck
                    (if (Variables|BPSetVariables|GetReturnToSender)
                      (Collision|IgnoreActorwhenMoving (Variables|Collision|GetCapsuleComponent) _hitactor true) ; If we didn't destroy the wall, pass through it
                      (CallFunction|HandleProjectileStopReturn)
                      (else
                        (CallFunction|ReturntoOwner))))
                  (:CastFailed
                    ;;[1 1 0] Hit a SimpleDesctructiblePropGroup Actor
                    (bind _assimple_destructible_prop_group (Utilities|Casting|CastToSimpleDestructiblePropGroup _hitactor)
                      (:then
                        (CallFunction|DamageEnvironmentActor _assimple_destructible_prop_group)
                        (if (Variables|BPSetVariables|GetReturnToSender)
                          (CallFunction|HandleProjectileStopReturn) ; continue returning to player
                          (else
                            (CallFunction|HandleProjectileStop) ; continue moving forward
                          )))
                      (:CastFailed
                        (Development|PrintString (Utilities|String|Append "Hit a destructible, non-building actor " _returnvalue))))))))
            (else
              ;;[1 0.6685 0.2208] Hit indestructible
              ;; If already returning, keep going. If not, return
              (Development|PrintString (Utilities|String|Append "Hit Indestructible: " _returnvalue))
              (if (Variables|BPSetVariables|GetReturnToSender)
                (Collision|IgnoreActorwhenMoving (Variables|Collision|GetCapsuleComponent) _hitactor true)
                (CallFunction|HandleProjectileStopReturn)
                (else
                  (CallFunction|GCFeedback (Variables|Default|GetGC_EnvImpact) (Transformation|GetActorLocation self) (Transformation|GetActorForwardVector self) _hitactor)
                  (CallFunction|ReturntoOwner))))))))))

(event Custom|NaturalApexReached
  ;[1 1 1] Return To Owner
  (Development|PrintString "Reached throw Apex")
  (CallFunction|ReturntoOwner))

(event Custom|CallReturnMovement
  ;[1 1 1] Call Return Movement
  (CallFunction|ResumeMovement false) ; Do not reset the velocity when returning
  (CallFunction|PlayImpactFeedback))

(event Custom|AnimTrailNotifyBegin
  (params (bActive :type Boolean))
  ;[1 1 1] Niagara Anim Trail Activation
  (Niagara|SetNiagaraVariable(Bool) (Variables|Default|GetNS_AnimTrail) "AnimTrailActivate" bActive))

(event Custom|PlayImpactFeedback
  ;[1 1 1] Trigger controller feedback
  (Networking|CanExecuteCosmeticEvents
    (:True
      (bind _projectileownerpawn (Variables|Default|GetProjectileOwnerPawn)
        (:then
          (if (Pawn|IsLocallyControlled _projectileownerpawn)
            (ForceFeedback|SpawnForceFeedbackAttached "/Game/Weapons/ForceFeedback/FF_Generic_Impact_Near.FF_Generic_Impact_Near" (Variables|/Script/Engine.Actor|GetRootComponent _projectileownerpawn))))
        (:else)))))

(event Custom|ReturnToOwner
  (bind _playerpawn (Variables|BPSetVariables|GetPlayerPawn))
  ;[1 1 1] Return To Owner
  (Networking|SwitchHasAuthority
    (:Authority
      (Development|PrintString "Return to Owner Triggered")
      (CallFunction|HandleProjectileStopReturn)
      (Utilities|IsValid _playerpawn
        (:Is%20Valid
          (bind _distance (Class|MissionBlueprintFunctionLibrary|DistanceBetweenTwoVectors (Transformation|GetActorLocation) (Transformation|GetActorLocation _playerpawn)))
          (if (>= _distance (ScalableFloat|GetValueatLevel (Variables|BPSetVariables|GetRow_ReturnHomingLockTargetDistanceThreshold)))
            (Variables|BPSetVariables|SetIsReturning true)
            (CallFunction|TurnonRecallFX)
            (Variables|BPSetVariables|SetReturnToSender true) ; Set Return Things
            (Networking|ForceNetUpdate)
            (else
              _)))
        (:Is%20Not%20Valid
          _))
      (CallFunction|KillProjectile))))

(event Custom|CallMovement
  ;[1 1 1] Call Outgoing Movement
  (CallFunction|ResumeMovement true)
  (CallFunction|PlayImpactFeedback))

(event Custom|AnimTrailNotifyEnd
  ;[1 1 1] Niagara Anim Trail Activation
  (Niagara|SetNiagaraVariable(Bool) (Variables|Default|GetNS_AnimTrail) "AnimTrailActivate" false))

(event Custom|TurnOnRecallFX
  (bind _rotatingmovement (Variables|Default|GetRotatingMovement))
  ;[1 1 1] Turn on return FX and sounds
  (Utilities|FlowControl|DoOnce false
    (:Completed
      (Development|PrintString "Recall FX")
      (CallFunction|PlayRecallSound)
      (CallFunction|AnimTrailNotifyEnd)
      ;[1 1 1] Rotate saber back to hand catch orientation
      (Components|Timeline|Stop (Variables|B_Prj_Realm_ThrownLobster|GetRotationSlowdown)) ; stop all ongoing timelines
      (Variables|/Script/Engine.RotatingMovementComponent|SetRotationRate_Roll (0.0 0.0 0.0) _rotatingmovement)
      (Components|Movement|StopMovementImmediately _rotatingmovement) ; stop rotation
      (Variables|Default|SetcachedMeshRotation (Transformation|GetWorldRotation (Variables|Default|GetSkelMesh_Weapon)))
      (Components|Timeline|SetPlayRate (Variables|B_Prj_Realm_ThrownLobster|GetReturnToHandMotion) (ScalableFloat|GetValueatLevel (Variables|BPSetVariables|GetRow_ReturnRotationTimelineRate)))
      (bind (_direction _rotation) (Timeline|ReturnToHandMotion PlayFromStart)
        (:Update
          (Transformation|SetWorldRotation (Variables|Default|GetSkelMesh_Weapon) (Math|Rotator|Lerp(Rotator) (Variables|Default|GetcachedMeshRotation) (Math|Rotator|FindLookatRotation (Transformation|GetActorLocation) (Transformation|GetSocketLocation (Variables|/Script/Engine.Character|GetMesh (Variables|Default|GetProjectileOwnerPawn)) "weapon_r")) _rotation) false false))
        (:turnOffBlade
          ;[1 0 0.21] Turn off Blade; this needs to be beefed up with the fancy saber turn-off FX in the weapon
          (if (Variables|Default|GetTurnOffBladeOnReturn?)
            (Rendering|SetVisibility (Variables|Default|Getblade) false false)
            (Collision|SetCollisionEnabled (Variables|Collision|GetCapsuleComponent) "NoCollision") ; REDUNDANT; most of this is in the initial recall now
          ))))))

(event EventOnTouched
  (params (OtherActor :type Actor%20Object%20Reference) (OtherComp :type Primitive%20Component%20Object%20Reference) (HitResult :type Hit%20Result%20Structure%20%28by%20ref%29) (bIsOverlap :type Boolean))
  ;[0.4662 1 0.9235] Event OnTouched - If overlap, call HandleOverlap to potentially damage a character
  ; If not an overlap, cache the velocity for use in the on stop event
  (if bIsOverlap
    (Variables|Default|SetHitResult HitResult)
    (CallFunction|HandleOverlap OtherActor)
    (else
      (Variables|BPSetVariables|SetPreviousVelocity (Transformation|GetVelocity)) ; Set Previous Velocity so on ResumeMovement we know the direction to go.\nSetting it here because this gets triggered before the stop event
    )))

(event Custom|DamageEnvironmentActor
  (params (Actor :type Actor%20Object%20Reference))
  (bind _returnvalue (Ability|GetAbilitySystemComponent Actor))
  ;[1 1 1] Damage building/environment actor
  (CallFunction|GCFeedback (Variables|Default|GetGC_EnvImpact) (Transformation|GetActorLocation self) (Transformation|GetActorForwardVector self) Actor)
  (bind _bdidactorhavetags (Class|AthenaQuestFunctionLibrary|AthenaCheckActorHasTags Actor (Variables|Default|GetInstaKillTags)))
  (Utilities|IsValid _returnvalue
    (:Is%20Valid
      (CallFunction|AddHitResulttoEffectContext)
      (GameplayEffects|ApplyGameplayEffectToTarget (Ability|GetAbilitySystemComponent (Variables|BPSetVariables|GetPlayerPawn)) (select _bdidactorhavetags (Variables|Default|GetProjectileEnvDamage_InstaKill) (Variables|Default|GetProjectileEnvDamage)) _returnvalue 0.0 (Variables|Default|GetHitEffectContext)))))

(event Custom|HandleOverlap
  (params (OverlapActor :type Actor%20Object%20Reference))
  ;[0.5948 0.9165 1] Overlap - Potentially damage character
  (Networking|SwitchHasAuthority
    (:Authority
      (bind _ascharacter (Utilities|Casting|CastToCharacter OverlapActor))
      ;;[1 1 1] ignore the hit if player was recently hit
      (bind _isimmune (CallFunction|CheckImmunityStatus _ascharacter))
      (if (not _isimmune)
        (if (not (Math|Boolean|ORBoolean (Utilities|Equal(Object) (Variables|Actor|GetInstigator) _ascharacter) (Team|OnSameTeam _ascharacter (Variables|BPSetVariables|GetPlayerPawn)) (Utilities|Array|ContainsItem (Actor|GetAttachedActors _ascharacter true false) (Variables|Actor|GetInstigator))))
          (CallFunction|EvaluateEndingHomingBehavior _ascharacter) ; stop homing if we reached the target
          (CallFunction|DamageHitActor _ascharacter)
          (Utilities|Map|Add (Variables|Default|GetRecentlyHit) _ascharacter (Fortnite|Sync|Time|GetServerWorldTimeSeconds)))
        (else
          (Development|PrintString "Immune from Attack, ignoring"))))))

(event Custom|DamageHitActor
  (params (Actor :type Actor%20Object%20Reference))
  (bind (_case_0_true _case_1_true) (GameplayTags|TagContainer|CompareTagAssetInterfacetoOtherTags (Variables|Default|GetInVehicle) (Variables|Default|GetInVehicle_SeatOpen) (Ability|GetAbilitySystemComponent Actor)))
  (bind _returnvalue (Game|GetInstigator))
  (bind _returnvalue_2 (Ability|GetAbilitySystemComponent _returnvalue))
  (bind _returnvalue_3 (Ability|GetAbilitySystemComponent Actor))
  ;[0.4688 0.3006 0.1296] Damage Hit Actor
  (Utilities|IsValid Actor
    (:Is%20Valid
      (bind _owningrealmweapon (Variables|Default|GetOwningRealmWeapon)
        (:then
          (Class|BAthenaGalileoLobsterRealm|HandleProjectileHit _owningrealmweapon))
        (:else
          _))
      (CallFunction|GCFeedback (Variables|Default|GetGC_PlayerImpact) (Transformation|GetActorLocation self) (Transformation|GetActorForwardVector self) Actor)
      (CallFunction|LaunchHitActor Actor)
      (Utilities|IsValid _returnvalue_3
        (:Is%20Valid
          (Utilities|IsValid _returnvalue
            (:Is%20Valid
              (Utilities|IsValid _returnvalue_2
                (:Is%20Valid
                  (if (or (not _case_0_true) (and _case_0_true _case_1_true))
                    (CallFunction|AddHitResulttoEffectContext)
                    (GameplayEffects|ApplyGameplayEffectToTarget _returnvalue_2 (Variables|Default|GetGE_Damage) _returnvalue_3 0.0 (Variables|Default|GetHitEffectContext))))))))))))

(event Custom|LaunchHitActor
  (params (Actor :type Actor%20Object%20Reference))
  (bind _returnvalue (CallFunction|CalculateKnockbackVelocity Actor))
  ;[1 1 1] Launch Actor
  (bind _ascharacter (Utilities|Casting|CastToCharacter Actor)
    (:then
      (bind _returnvalue_2 (GameplayTags|HasMatchingGameplayTag(Message) _ascharacter "(TagName=\"Gameplay.Status.KnockbackImmunity\")"))
      (if (not _returnvalue_2)
        (Character|LaunchCharacter _ascharacter _returnvalue true true)))
    (:CastFailed
      (bind _asfort_vehicle_interface (Utilities|Casting|CastToFortVehicleInterface Actor))
      (bind _returnvalue_3 (Vehicle|GetAllPassengers(Message) (Utilities|ToObject(Interface) _asfort_vehicle_interface)))
      (if (not (Utilities|Array|ContainsItem _returnvalue_3 (Variables|BPSetVariables|GetPlayerPawn)))
        (Audio|BroadcastAbilityImpactatLocation 0 (Utilities|ToObject(Interface) _asfort_vehicle_interface) _returnvalue 777.0) ; This launches vehicles (named weird for anti-cheat reasons)\nPosition = velocity to launch it at\nRadius must equal 777\nGates knows more about it
      ))))

(event Custom|GCFeedback
  (params (GameplayCueTag :type Gameplay%20Tag%20Structure) (Location :type Vector) (Normal :type Vector) (SourceObject :type Object%20Reference))
  ;[1 0.8018 0] GAMEPLAY CUE FEEDBACK
  (GameplayCue|ExecuteGameplayCueOnActor(Burst) self GameplayCueTag (Ability|GameplayCue|MakeGameplayCueParameters :NormalizedMagnitude 0.0 :RawMagnitude 0.0 :Location Location :Normal Normal :Instigator self :EffectCauser self :SourceObject SourceObject :GameplayEffectLevel 1 :AbilityLevel 1 :bReplicateLocationWhenUsingMinimalRepProxy false)))

(event ProjectileMovement|EventOnReachedLockTargetDistanceThreshold
  ;[1 1 1] EVENT -> Reached Homing Target
  (if (Variables|BPSetVariables|GetIsReturning)
    (Utilities|FlowControl|DelayUntilNextTick) ; This is called within the projectile component \ntick while loop, so stopping simulation \nthis frame will cause an access crash
    (CallFunction|KillProjectile)))

(event ProjectileMovement|EventOnHomingPassedTarget
  ;[1 1 1] EVENT -> Reached Homing Target
  (if (Variables|BPSetVariables|GetIsReturning)
    (Projectile|SetHomingTurnSpeedMax 0 0 5e+04 :!class /Script/FortniteGame.FortProjectileMovementInterface)))

(event Custom|KillProjectile
  (bind _projectilemovementcomponent (Variables|Movement|GetProjectileMovementComponent))
  (bind _playerpawn (Variables|BPSetVariables|GetPlayerPawn))
  ;[1 0.7007 0.4203] Kill Projectile (Called from Pawn's OnDied and when it reaches its homing threshold)
  (Utilities|FlowControl|DoOnce false
    (:Completed
      (Utilities|FlowControl|Sequence
        (:then_0
          (Components|Movement|StopMovementImmediately _projectilemovementcomponent)
          (Game|Components|ProjectileMovement|StopSimulating _projectilemovementcomponent (Collision|MakeHitResult false false 0.0 0.0 (Transformation|GetActorLocation)))
          (Rendering|SetVisibility (Variables|Default|GetScene) false true)
          (Development|PrintString "Stopped and Hid Projectile"))
        (:then_1
          (Utilities|IsValid _playerpawn
            (:Is%20Valid
              (Utilities|IsValid (Ability|GetAbilitySystemComponent _playerpawn)
                (:Is%20Valid
                  (Ability|SendGameplayEventtoActor (Variables|BPSetVariables|GetPlayerPawn) (Variables|BPSetVariables|GetEvent_Catch)) ; Trigger Catch GA
                )
                (:Is%20Not%20Valid
                  _))
              (SoundIndicator|StopSoundIndicatorLooping (Variables|Default|GetFortSoundIndicator))
              (bind _recallaudio (Variables|Default|GetRecallAudio)
                (:then
                  (Audio|Components|Audio|FadeOut _recallaudio 0.35 0.0 "Linear" :!class /Script/Engine.AudioComponent))
                (:else
                  _))
              (Networking|SwitchHasAuthority
                (:Authority
                  (bind _owningrealmweapon (Variables|Default|GetOwningRealmWeapon)
                    (:then
                      (Class|BAthenaGalileoLobsterRealm|SetProjectile _owningrealmweapon 0))
                    (:else)))))
            (:Is%20Not%20Valid
              _)))
        (:then_2
          ;;[1 1 1] Call ProjectileDestroyed directly so we can start processing cleanup. However, call Kill instead of DestroyActor to set the lifespan to a small duration. This gives time for the server to finish replicating
          (CallFunction|ProjectileDestroyed self)
          (Networking|SwitchHasAuthority
            (:Authority
              (Delivery|Kill))))))))

(event Custom|ProjectileDestroyed
  (params (DestroyedActor :type Actor%20Object%20Reference))
  (bind _playerpawn (Variables|BPSetVariables|GetPlayerPawn))
  (bind _returnvalue (Ability|GetAbilitySystemComponent _playerpawn))
  (bind _owningrealmweapon (Variables|Default|GetOwningRealmWeapon))
  ;[1 0.3227 0.1441] ProjectileDestroyed (called from KillProjectile directly and also bound to OnDestroyed)
  (Utilities|FlowControl|DoOnce false
    (:Completed
      (Development|PrintString "ProjectileDestroyed")
      (Utilities|IsValid _playerpawn
        (:Is%20Valid
          (Utilities|IsValid _returnvalue
            (:Is%20Valid
              (GameplayEffects|RemoveActiveGameplayEffect _returnvalue (Variables|BPSetVariables|GetQuestCheckTag_GE) -1)
              ;;[0.9734 0.5271 0.1912] Remove GE_Realm_ProjectileTracker and "realm" holster ID. This was originally handled in the catch GA, but moved to here as the catch GA can be removed from the player if they switch weapons.
              (GameplayEffects|RemoveActiveEffectswithGrantedTags _returnvalue "(GameplayTags=((TagName=\"Granted.Galileo.Lobster.Unusable\")))")
              (Weapon|SetHolsterWeapon (Variables|BPSetVariables|GetPlayerPawn) 0 false false "realm")
              ;;[0.9734 0.5271 0.1912] If the owning lightsaber weapon IS NOT equipped, disable Instant On. (So the full effects play next time the lighsaber is equipped.)
              ;; If the owning lightsaber weapon IS equipped, GA_Realm_Catch will handle this after the catch is complete.
              (Utilities|IsValid _owningrealmweapon
                (:Is%20Valid
                  (if (not (FortWeapon|IsEquipped _owningrealmweapon))
                    (Variables|/Galileo/Items/Lobster/Actors/B_Athena_Galileo_Lobster_Parent.B_Athena_Galileo_Lobster_Parent_C|SetInstantOn false _owningrealmweapon)))))))))))

(fn ACloserThanB?
  (params (A :type Actor%20Object%20Reference) (B :type Actor%20Object%20Reference))
  (return (< (Class|AbilityFunctions|GetDistanceBetweenActors A (Actor|GetOwner)) (Class|AbilityFunctions|GetDistanceBetweenActors B (Actor|GetOwner)))))

(fn AddHitResultToEffectContext
  (bind _output_get (Variables|Default|SetHitEffectContext (Class|AthenaFunctionLibrary|MakeSimpleFortEffectContext (Game|GetInstigator) self)))
  (Ability|EffectContext|AddHitResult _output_get (Variables|Default|GetHitResult)))

(fn CalculateKnockbackVelocity
  (params (Actor :type Actor%20Object%20Reference))
  (bind (_ascharacter _bsuccess) (Utilities|Casting|CastToCharacter Actor))
  (return (Math|Vector|MakeVector (.x (* (Math|Vector|Normalize (Math|Vector|GetUnitDirection(Vector) (Transformation|GetActorLocation) (Transformation|GetActorLocation Actor))) (Math|Float|SelectFloat (ScalableFloat|GetValueatLevel (Variables|BPSetVariables|GetRow_PlayerKnockbackStrength)) (ScalableFloat|GetValueatLevel (Variables|BPSetVariables|GetRow_VehicleKnockbackStrength)) _bsuccess))) (.y (* (Math|Vector|Normalize (Math|Vector|GetUnitDirection(Vector) (Transformation|GetActorLocation) (Transformation|GetActorLocation Actor))) (Math|Float|SelectFloat (ScalableFloat|GetValueatLevel (Variables|BPSetVariables|GetRow_PlayerKnockbackStrength)) (ScalableFloat|GetValueatLevel (Variables|BPSetVariables|GetRow_VehicleKnockbackStrength)) _bsuccess))) (+ (.z (* (Math|Vector|Normalize (Math|Vector|GetUnitDirection(Vector) (Transformation|GetActorLocation) (Transformation|GetActorLocation Actor))) (Math|Float|SelectFloat (ScalableFloat|GetValueatLevel (Variables|BPSetVariables|GetRow_PlayerKnockbackStrength)) (ScalableFloat|GetValueatLevel (Variables|BPSetVariables|GetRow_VehicleKnockbackStrength)) _bsuccess))) (Math|Float|SelectFloat (ScalableFloat|GetValueatLevel (Variables|BPSetVariables|GetRow_PlayerKnockbackStrength)) (ScalableFloat|GetValueatLevel (Variables|BPSetVariables|GetRow_VehicleKnockbackStrength)) _bsuccess)))))

(fn CalculateNewVelocity
  (params (TraceStart :type Vector) (TraceEnd :type Vector))
  (return (* (Math|Vector|Normalize (- TraceEnd TraceStart)) (* (Math|Vector|VectorLength (Variables|BPSetVariables|GetPreviousVelocity)) 0.9)) (>= (* (Math|Vector|VectorLength (Variables|BPSetVariables|GetPreviousVelocity)) 0.9) 2e+03)))

(fn CanApplyHolster
  (params (HolsterId :type Name%20%28by%20ref%29 :default "None"))
  (return true))

(fn CheckImmunityStatus
  (params (Key :type Character%20Object%20Reference%20%28by%20ref%29))
  (bind _recentlyhit (Variables|Default|GetRecentlyHit))
  (bind (_value _returnvalue) (Utilities|Map|Find _recentlyhit Key))
  (if _returnvalue
    (if (> (- (Fortnite|Sync|Time|GetServerWorldTimeSeconds) _value) (ScalableFloat|GetValueatLevel (Variables|BPSetVariables|GetRow_PostHitImmunityDuration)))
      (Development|PrintString "Immunity expired, removing from map")
      (Utilities|Map|Remove (Variables|Default|GetRecentlyHit) Key)
      (return false)
      (else
        (return true)))
    (else
      (return false))))

(fn ChooseHomingTargetOffset
  (params (TargetActor :type Character%20Object%20Reference))
  (Utilities|IsValid TargetActor
    (:Is%20Valid
      (bind _returnvalue (GameplayTags|HasMatchingGameplayTag(Message) (Variables|GetTargetActor) "(TagName=\"Pawn.Athena.NPC.Wildlife\")"))
      (if _returnvalue
        ;[0.7898 1 0.48] Wildlife
        (return (Variables|HomingValues|GetWildlifeHomingTargetOffset))
        (else
          (bind _asfort_player_pawn (Utilities|Casting|CastToFortPlayerPawn (Variables|GetTargetActor))
            (:then
              (if (Health|IsDBNO _asfort_player_pawn)
                ;[1 0.593 0.3247] DBNO
                (return (Variables|HomingValues|GetDBNOHomingTargetOffset))
                (else
                  _)))
            (:CastFailed
              _))
          ;[0.5832 0.9901 1] Standard
          (return (Variables|HomingValues|GetHomingTargetOffset)))))
    (:Is%20Not%20Valid
      (return "0, 0, 0"))))

(fn EvaluateEndingHomingBehavior
  (params (HitPawn :type Character%20Object%20Reference))
  (bind (_homingstyle _minturnspeed _maxturnspeed _ramptimefrommintomaxturnspeed _locktargetdistancethreshold _randomseed _lockedontarget) (Utilities|Struct|BreakProjectileHomingData (Variables|Default|GetProjectile%20Homing%20Data)))
  (bind _row_homingeffecttriggerrange (Variables|BPSetVariables|GetRow_HomingEffectTriggerRange))
  (bind _returnvalue (ScalableFloat|GetValueatLevel _row_homingeffecttriggerrange))
  (bind _row_homingminturnspeed (Variables|BPSetVariables|GetRow_HomingMinTurnSpeed))
  (bind _returnvalue_2 (ScalableFloat|GetValueatLevel _row_homingminturnspeed))
  (bind _row_homingmaxturnspeed (Variables|BPSetVariables|GetRow_HomingMaxTurnSpeed))
  (bind _returnvalue_3 (ScalableFloat|GetValueatLevel _row_homingmaxturnspeed))
  (bind _row_homingturnramptime (Variables|BPSetVariables|GetRow_HomingTurnRampTime))
  (bind _returnvalue_4 (ScalableFloat|GetValueatLevel _row_homingturnramptime))
  (bind _projectilehomingdata (Utilities|Struct|MakeProjectileHomingData "LockOn" _returnvalue_2 _returnvalue_3 _returnvalue_4 _returnvalue))
  (bind _projectilemovementcomponent (Variables|Movement|GetProjectileMovementComponent))
  ;[1 1 1] stop chasing the target if we've hit it
  (if (== HitPawn _lockedontarget)
    (Projectile|SetHomingData _projectilemovementcomponent _projectilehomingdata :!class /Script/FortniteGame.FortProjectileMovementComponent) ; reset the homing parameters to stop following the target
    (Variables|Default|SetProjectile%20Homing%20Data _projectilehomingdata)
    (Components|Movement|SetUpdatedComponent _projectilemovementcomponent (Variables|Collision|GetCapsuleComponent))))

(fn FindHomingTarget
  (bind (_outlocation _outrotation) (Actor|GetActorEyesViewPoint (Actor|GetOwner)))
  ;[1 1 1] Shape to collect potential homing targets
  (bind _outactors (Collision|BoxOverlapActors (Math|Transform|MakeTransform (+ (* (/ (ScalableFloat|GetValueatLevel (Variables|BPSetVariables|GetRow_HomingTargetingMaxRange)) 2.0) (Math|Vector|GetForwardVector _outrotation)) _outlocation) _outrotation) (Math|Vector|MakeVector (/ (ScalableFloat|GetValueatLevel (Variables|BPSetVariables|GetRow_HomingTargetingMaxRange)) 2.0) (ScalableFloat|GetValueatLevel (Variables|BPSetVariables|GetRow_HomingTargetingRadius)) (ScalableFloat|GetValueatLevel (Variables|BPSetVariables|GetRow_HomingTargetingRadius))) (Utilities|Array|MakeArray "ObjectTypeQuery3" "ObjectTypeQuery7") 0 (Utilities|Array|MakeArray (Actor|GetOwner)) :!class /Script/FortniteGame.FortKismetLibrary))
  ;[1 1 1] Finds the closest valid actor
  (bind _asfort_player_pawn (Utilities|Casting|CastToFortPlayerPawn (Actor|GetOwner)))
  (for _array_element _outactors
    (bind _ascharacter (Utilities|Casting|CastToCharacter _array_element))
    (if (> (Class|AbilityFunctions|GetDistanceBetweenActors _ascharacter _asfort_player_pawn) (ScalableFloat|GetValueatLevel (Variables|BPSetVariables|GetRow_HomingTargetingMinRange)))
      (if (not (Team|OnSameTeam _ascharacter _asfort_player_pawn))
        (bind _returnvalue (GameplayTags|HasMatchingGameplayTag(Message) _ascharacter "(TagName=\"Pawn.Athena.NPC.Wildlife\")"))
        (if _returnvalue
          ;[1 1 1] Wildlife
          (LocalVariables|GetTargetWildlife
            (:then
              (if (CallFunction|ACloserThanB? _ascharacter (LocalVariables|GetTargetPlayer))
                _))
            (:else
              _))
          (bind (_outhit _returnvalue_2) (Collision|LineTraceByProfile (Transformation|GetActorLocation) (Physics|GetCenterOfMass (Variables|/Script/Engine.Character|GetCapsuleComponent _ascharacter) "None") "FortProjectileOverlapPawns"))
          (if (not _returnvalue_2)
            (LocalVariables|SetTargetWildlife _ascharacter))
          (else
            (LocalVariables|GetTargetPlayer
              (:then
                (if (CallFunction|ACloserThanB? _ascharacter (LocalVariables|GetTargetPlayer))
                  _))
              (:else
                _))
            (bind (_outhit_2 _returnvalue_3) (Collision|LineTraceByProfile :Start (Transformation|GetActorLocation) :End (Physics|GetCenterOfMass (Variables|/Script/Engine.Character|GetCapsuleComponent _ascharacter) "None") :ProfileName "FortProjectileOverlapPawns" :bTraceComplex false :DrawDebugType "Persistent" :bIgnoreSelf true :TraceColor "(R=1.000000,G=0.000000,B=0.000000,A=1.000000)" :TraceHitColor "(R=0.000000,G=1.000000,B=0.000000,A=1.000000)" :DrawTime 5.0))
            (if (not _returnvalue_3)
              (LocalVariables|SetTargetPlayer _ascharacter)))))))
  ;[1 1 1] Return player; if none return wildlife
  (bind _targetplayer (LocalVariables|GetTargetPlayer)
    (:then
      (Development|PrintString (Utilities|String|Append "Locking On to Character: " (Utilities|GetDisplayName (LocalVariables|GetTargetPlayer))))
      (return (LocalVariables|GetTargetPlayer)))
    (:else
      (bind _targetwildlife (LocalVariables|GetTargetWildlife)
        (:then
          (Development|PrintString (Utilities|String|Append "Locking On to Wildlife: " (Utilities|GetDisplayName (LocalVariables|GetTargetPlayer))))
          (return (LocalVariables|GetTargetWildlife)))
        (:else
          (return 0))))))

(fn GetAffiliationsToShowFor
  (return (LocalVariables|GetAffiliation)))

(fn HandleProjectileStopReturn
  (Networking|SwitchHasAuthority
    (:Authority
      (Utilities|Time|SetTimerbyEvent (EventDispatchers|CreateEvent :!event CallReturnMovement) 1e-05))))

(fn HandleProjectileStop
  (Networking|SwitchHasAuthority
    (:Authority
      (Utilities|Time|SetTimerbyEvent (EventDispatchers|CreateEvent :!event CallMovement) 1e-05))))

(fn OnRep_ReturnToSender
  (bind _self self)
  (bind _projectilemovementcomponent (Variables|Movement|GetProjectileMovementComponent))
  (Utilities|Time|ClearandInvalidateTimerbyHandle (Variables|Default|GetNaturalApexReachedTimerHandle))
  (Development|PrintString "REP NOTIFY: Return To Sender ")
  (Collision|SetCollisionEnabled (Variables|Collision|GetCapsuleComponent) "NoCollision")
  (Projectile|SetHomingStyle 0 _self "LockOn" false :!class /Script/FortniteGame.FortProjectileMovementInterface)
  (Projectile|SetHomingTarget 0 _self (Variables|BPSetVariables|GetPlayerPawn) 0 :!class /Script/FortniteGame.FortProjectileMovementInterface)
  (Projectile|SetHomingLockTargetDistanceThreshold 0 _self (ScalableFloat|GetValueatLevel (Variables|BPSetVariables|GetRow_ReturnHomingLockTargetDistanceThreshold)) :!class /Script/FortniteGame.FortProjectileMovementInterface)
  (Projectile|SetHomingTargetOffset _self (Variables|HomingValues|GetReturnHomingTargetOffset) :!class /Script/FortniteGame.FortProjectileMovementInterface)
  (Variables|/Script/Engine.ProjectileMovementComponent|SetInitialSpeed (ScalableFloat|GetValueatLevel (Variables|BPSetVariables|GetRow_ReturnStartVelocity)) _projectilemovementcomponent)
  (Variables|/Script/Engine.ProjectileMovementComponent|SetMaxSpeed (ScalableFloat|GetValueatLevel (Variables|BPSetVariables|GetRow_ReturnEndVelocity)) _projectilemovementcomponent)
  (Projectile|SetHomingTurnSpeedMax 0 self (ScalableFloat|GetValueatLevel (Variables|BPSetVariables|GetRow_ReturnMaxTurnSpeed)) :!class /Script/FortniteGame.FortProjectileMovementInterface)
  (Collision|SetCollisionEnabled (Variables|Collision|GetCapsuleComponent) "QueryOnly")
  (CallFunction|ResumeMovement false) ; Do not reset the velocity when returning
  (Variables|Default|SetProjectile%20Homing%20Data (Utilities|Struct|MakeProjectileHomingData "None" (ScalableFloat|GetValueatLevel (Variables|BPSetVariables|GetRow_HomingMinTurnSpeed)) (ScalableFloat|GetValueatLevel (Variables|BPSetVariables|GetRow_HomingMaxTurnSpeed)) (ScalableFloat|GetValueatLevel (Variables|BPSetVariables|GetRow_HomingTurnRampTime)) (ScalableFloat|GetValueatLevel (Variables|BPSetVariables|GetRow_HomingEffectTriggerRange)))))

(fn PlayRecallSound
  (bind _playerpawn (Variables|BPSetVariables|GetPlayerPawn))
  (bind _playerpawn_2 (Variables|BPSetVariables|GetPlayerPawn))
  (Utilities|FlowControl|Sequence
    (:then_0
      ;[1 1 1] Play Recall Whoosh Sound
      (bind _returnvalue (Audio|SpawnSoundAttached (Variables|Default|GetRecallWhooshSound) (Variables|Collision|GetCapsuleComponent)))
      (Utilities|IsValid _playerpawn_2
        (:Is%20Valid
          (Audio|Parameter|SetBooleanParameter _returnvalue "IsLocal" (Pawn|IsLocallyViewed _playerpawn_2) :!class /Script/AudioExtensions.AudioParameterControllerInterface))))
    (:then_1
      ;[1 1 1] Stop cast sound so we can then play recall sound. Valid checks just incase of any kind of spam.
      (bind _castinairaudio (Variables|Default|GetCastInAirAudio)
        (:then
          (Audio|Components|Audio|FadeOut _castinairaudio 0.5 0.0 "Linear" :!class /Script/Engine.AudioComponent))
        (:else
          _))
      (Variables|Default|GetRecallAudio
        (:then)
        (:else
          (bind _returnvalue_2 (Audio|SpawnSoundAttached (Variables|Sounds|GetRecallSound) (Variables|Collision|GetCapsuleComponent)))
          (bind _output_get (Variables|Default|SetRecallAudio _returnvalue_2))
          (Utilities|IsValid _playerpawn
            (:Is%20Valid
              (Audio|Parameter|SetBooleanParameter _output_get "IsLocal" (Pawn|IsLocallyViewed _playerpawn) :!class /Script/AudioExtensions.AudioParameterControllerInterface))))))))

(fn Resume_Movement
  (params (SetVelocityToPrevious :type Boolean))
  (Projectile|ResumeSimulation)
  (Collision|SetCollisionEnabled (Variables|Collision|GetCapsuleComponent) "QueryOnly" :!pinned true) ; Make sure the collision is enabled
  (Components|Movement|SetUpdatedComponent (Variables|Movement|GetProjectileMovementComponent) (Variables|Collision|GetCapsuleComponent))
  (if SetVelocityToPrevious
    (Variables|/Script/Engine.MovementComponent|SetVelocity (Variables|BPSetVariables|GetPreviousVelocity) (Variables|Movement|GetProjectileMovementComponent))))

(fn ShouldShowSoundIndicator
  (params (PlayerController :type Fort%20Player%20Controller%20Object%20Reference))
  (return (!= PlayerController (Fortnite|GetFortPlayerControllerfromActor (Game|GetInstigator)))))

(fn ConstructionScript
  (bind _capsulecomponent (Variables|Collision|GetCapsuleComponent))
  (Components|Capsule|SetCapsuleHalfHeight _capsulecomponent (ScalableFloat|GetValueatLevel (Variables|BPSetVariables|GetRow_PlayerCollisionHalfHeight)) true)
  (Components|Capsule|SetCapsuleRadius _capsulecomponent (ScalableFloat|GetValueatLevel (Variables|BPSetVariables|GetRow_PlayerCollisionRadius)) true)
  (|CallParentFunction))
