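; Lobster Realm saber block ability that lets the player throw the saber at a homing target by pressing primary while it is active.
; It applies a throw effect, pushes a camera mode, shows the prompt and reticle, and undoes these on end.
; A repeating server timer box-scans the view for non-teammates, using aim angle and line traces, preferring players over wildlife.
; The press sends the throw event with the weapon's stored target.
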(event Ability|EventActivateAbility
  (|CallParentFunction)
  (bind _returnvalue (Ability|ApplyGameplayEffectToOwner (Variables|Default|GetGE%20throw)))
  (Variables|Default|SetThrow%20Effect%20Handle _returnvalue)
  (Ability|Tasks|WaitGameplayEvent "(TagName=\"Event.Primary.Pressed\")"
    (:then
      (Camera|PushCameraMode (Variables|Default|GetCamera%20Mode))
      (bind _playerpawn (Variables|Default|GetPlayerPawn)
        (:then
          (bind _asb_athena_galileo_lobster_realm (Utilities|Casting|CastToB_Athena_Galileo_Lobster_Realm (Weapon|GetCurrentWeapon _playerpawn)))
          (Variables|Default|SetWeapon _asb_athena_galileo_lobster_realm)
          (Utilities|IsServer
            (:true
              (bind _returnvalue_2 (Utilities|Time|SetTimerbyEvent (delegate Custom|FindTarget) 0.1 true))
              (Variables|Default|SetTargetTimer _returnvalue_2))))
        (:else)))
    (:EventReceived
      (Ability|SendGameplayEvent (Variables|Default|GetT_ThrowEvent) (Utilities|Struct|MakeGameplayEventData (Variables|Default|GetT_ThrowEvent) 0 0 (Variables|/Galileo/Items/Lobster/Actors/B_Athena_Galileo_Lobster_Realm.B_Athena_Galileo_Lobster_Realm_C|GetHomingTarget (Variables|Default|GetWeapon))))
      (Utilities|IsServer
        (:true
          (Variables|/Galileo/Items/Lobster/Actors/B_Athena_Galileo_Lobster_Realm.B_Athena_Galileo_Lobster_Realm_C|SetHomingTarget 0 (Variables|Default|GetWeapon))
          (Utilities|Time|ClearandInvalidateTimerbyHandle (Variables|Default|GetTargetTimer)))))))

(event Custom|FindTarget
  (bind _targetactor (CallFunction|FindHomingTarget))
  (bind _weapon (Variables|Default|GetWeapon)
    (:then
      (Variables|/Galileo/Items/Lobster/Actors/B_Athena_Galileo_Lobster_Realm.B_Athena_Galileo_Lobster_Realm_C|SetHomingTarget _targetactor _weapon)
      (Variables|Default|SetTargetPlayer 0)
      (Variables|Default|SetTargetWildlife 0))
    (:else)))

(event Ability|EventOnEndAbility
  (params (bWasCancelled :type Boolean))
  (|CallParentFunction bWasCancelled)
  (Ability|RemoveGameplayEffectFromOwnerWithHandle (Variables|Default|GetThrow%20Effect%20Handle))
  (Camera|PopCameraMode))

(fn ACloserThanB?
  (params (A :type Actor%20Object%20Reference) (B :type Actor%20Object%20Reference))
  (bind _playerpawn (Variables|Default|GetPlayerPawn)
    (:then
      (return (< (Class|AbilityFunctions|GetAngleBetweenTwoVectors (Math|Vector|Normalize (- (Transformation|GetActorLocation A) (Transformation|GetActorLocation _playerpawn)) 0.1) (+ (Math|Vector|GetForwardVector (.OutRotation (Actor|GetActorEyesViewPoint _playerpawn))) (* (Math|Vector|GetRightVector (.OutRotation (Actor|GetActorEyesViewPoint _playerpawn))) 0.05))) (Class|AbilityFunctions|GetAngleBetweenTwoVectors (Math|Vector|Normalize (- (Transformation|GetActorLocation B) (Transformation|GetActorLocation _playerpawn)) 0.1) (+ (Math|Vector|GetForwardVector (.OutRotation (Actor|GetActorEyesViewPoint _playerpawn))) (* (Math|Vector|GetRightVector (.OutRotation (Actor|GetActorEyesViewPoint _playerpawn))) 0.05))))))
    (:else)))

(fn EndAbilityCleanup
  (bind _realmweapon (Variables|Default|GetRealmWeapon))
  (|CallParentFunction)
  (Class|BAthenaGalileoLobsterRealm|ToggleSaberThrowPrompt _realmweapon false)
  (Class|BAthenaGalileoLobsterParent|ReticleStatusDisplay _realmweapon false))

(fn FindHomingTarget
  (bind (_outlocation _outrotation) (Actor|GetActorEyesViewPoint (Variables|Default|GetPlayerPawn)))
  (bind _returnvalue (CallFunction|GetPrjSpawnLocation))
  ;[1 1 1] Shape to collect potential homing targets
  (bind _outactors (Collision|BoxOverlapActors (Math|Transform|MakeTransform (+ (* (/ (ScalableFloat|GetValueatLevel (Variables|Default|GetRow_HomingTargetingMaxRange)) 2.0) (Math|Vector|GetForwardVector _outrotation)) _outlocation) _outrotation) (Math|Vector|MakeVector (/ (ScalableFloat|GetValueatLevel (Variables|Default|GetRow_HomingTargetingMaxRange)) 2.0) (ScalableFloat|GetValueatLevel (Variables|Default|GetRow_HomingTargetingRadius)) (ScalableFloat|GetValueatLevel (Variables|Default|GetRow_HomingTargetingRadius))) (Utilities|Array|MakeArray "ObjectTypeQuery3" "ObjectTypeQuery7") 0 (Utilities|Array|MakeArray (Variables|Default|GetPlayerPawn)) :!class /Script/FortniteGame.FortKismetLibrary))
  ;[1 1 1] Finds the closest valid actor
  (bind _playerpawn (Variables|Default|GetPlayerPawn)
    (:then
      (for _array_element _outactors
        (bind _ascharacter (Utilities|Casting|CastToCharacter _array_element))
        (if (> (Class|AbilityFunctions|GetDistanceBetweenActors _ascharacter _playerpawn) (ScalableFloat|GetValueatLevel (Variables|Default|GetRow_HomingTargetingMinRange)))
          (if (not (Team|OnSameTeam _ascharacter _playerpawn))
            (bind _returnvalue_2 (GameplayTags|HasMatchingGameplayTag(Message) _ascharacter "(TagName=\"Pawn.Athena.NPC.Wildlife\")"))
            (if _returnvalue_2
              ;[1 1 1] Wildlife
              (Variables|Default|GetTargetWildlife
                (:then
                  (if (CallFunction|ACloserThanB? _ascharacter (Variables|Default|GetTargetPlayer))
                    _))
                (:else
                  _))
              (bind (_outhit _returnvalue_3) (Collision|LineTraceByProfile _returnvalue (Physics|GetCenterOfMass (Variables|/Script/Engine.Character|GetCapsuleComponent _ascharacter) "None") "FortProjectileOverlapPawns" true))
              (if (not _returnvalue_3)
                (Variables|Default|SetTargetWildlife _ascharacter))
              (else
                (Variables|Default|GetTargetPlayer
                  (:then
                    (if (CallFunction|ACloserThanB? _ascharacter (Variables|Default|GetTargetPlayer))
                      _))
                  (:else
                    _))
                (bind (_outhit_2 _returnvalue_4) (Collision|LineTraceByProfile _returnvalue (Physics|GetCenterOfMass (Variables|/Script/Engine.Character|GetCapsuleComponent _ascharacter) "None") "FortProjectileOverlapPawns" true))
                (if (not _returnvalue_4)
                  (Variables|Default|SetTargetPlayer _ascharacter)))))))
      ;[1 1 1] Return player; if none return wildlife
      (bind _targetplayer (Variables|Default|GetTargetPlayer)
        (:then
          (Development|PrintString (Utilities|String|Append "Locking On to Character: " (Utilities|GetDisplayName (Variables|Default|GetTargetPlayer))))
          (return (Variables|Default|GetTargetPlayer)))
        (:else
          (bind _targetwildlife (Variables|Default|GetTargetWildlife)
            (:then
              (Development|PrintString (Utilities|String|Append "Locking On to Wildlife: " (Utilities|GetDisplayName (Variables|Default|GetTargetPlayer))))
              (return (Variables|Default|GetTargetWildlife)))
            (:else
              (return 0))))))
    (:else)))

(fn GetPrjSpawnLocation
  (bind _playerpawn (Variables|Default|GetPlayerPawn)
    (:then
      (return (+ (Transformation|GetActorLocation _playerpawn) (Math|Vector|vector+vector (* (Math|Vector|GetForwardVector (Pawn|GetControlRotation _playerpawn :!class /Script/Engine.Pawn)) (Variables|Default|GetSpawnPrjForwardOffset)) (* (Math|Vector|GetRightVector (Pawn|GetControlRotation _playerpawn :!class /Script/Engine.Pawn)) (Variables|Default|GetSpawnRightOffset)) (select (Variables|/Script/Engine.Character|GetbIsCrouched _playerpawn) (* (Math|Vector|GetUpVector (Pawn|GetControlRotation _playerpawn :!class /Script/Engine.Pawn)) (Variables|Default|GetSpawnUpOffsetCrouch)) (* (Math|Vector|GetUpVector (Pawn|GetControlRotation _playerpawn :!class /Script/Engine.Pawn)) (Variables|Default|GetSpawnUpOffset)))))))
    (:else)))

(fn InitAbility
  (|CallParentFunction)
  (bind _asb_athena_galileo_lobster_realm (Utilities|Casting|CastToB_Athena_Galileo_Lobster_Realm (Ability|GetCurrentSourceWeapon)))
  (bind _output_get (Variables|Default|SetRealmWeapon _asb_athena_galileo_lobster_realm))
  (Class|BAthenaGalileoLobsterRealm|ToggleSaberThrowPrompt _output_get true)
  (Class|BAthenaGalileoLobsterParent|ReticleStatusDisplay _output_get true))
