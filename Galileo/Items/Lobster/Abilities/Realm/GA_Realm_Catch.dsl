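; Gameplay ability for catching the Galileo Lobster Realm weapon, a lightsaber, extending a parent catch ability.
; On activation it checks whether the current weapon is trying to block and, if so, selects the catch-to-block montage before running the parent logic.
; On end it runs the parent, clears the saber's instant-on flag so later catches play the standard power-on animation, and pops the camera mode.
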(event Ability|EventActivateAbility
  (bind _outfortplayerpawn (CallFunction|GetAvatarasValidFortPlayerPawn)
    (:ValidFortPlayerPawn
      (bind _asb_athena_galileo_lobster_realm (Utilities|Casting|CastToB_Athena_Galileo_Lobster_Realm (Weapon|GetCurrentWeapon _outfortplayerpawn))
        (:then
          (if (Variables|/Galileo/Items/Lobster/Actors/B_Athena_Galileo_Lobster_Realm.B_Athena_Galileo_Lobster_Realm_C|GetbTryingToBlock _asb_athena_galileo_lobster_realm)
            (Variables|CatchVars|SetMontageToPlay (Variables|Default|GetCatchToBlockMontage))
            (else
              _))
          (|CallParentFunction))
        (:CastFailed
          _)))
    (:AvatarCastFailed
      _)))

(event Ability|EventOnEndAbility
  (params (bWasCancelled :type Boolean))
  (|CallParentFunction bWasCancelled)
  ;[1 1 1] sets the lightsaber back to non-instant turn on so that SUBSEQUENT catches will play the standard "power on" animation
  (bind _asb_athena_galileo_lobster_realm (Utilities|Casting|CastToB_Athena_Galileo_Lobster_Realm (Variables|/Script/FortniteGame.FortPawn|GetCurrentWeapon (Variables|CatchVars|GetPlayerPawn)))
    (:then
      (Variables|/Galileo/Items/Lobster/Actors/B_Athena_Galileo_Lobster_Parent.B_Athena_Galileo_Lobster_Parent_C|SetInstantOn false _asb_athena_galileo_lobster_realm))
    (:CastFailed
      _))
  (Camera|PopCameraMode))
