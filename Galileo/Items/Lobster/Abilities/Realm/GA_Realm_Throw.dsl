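; Gameplay ability that throws the Realm "Lobster" lightsaber as a homing projectile, triggered by gameplay event or AI custom event carrying the target.
; A montage notify spawns the projectile on the server after committing the ability, which prevents multi-spawn exploits.
; Spawn position uses riding, crouched or standing offsets, with a line trace pulling it back from nearby obstructions.
; Character targets get homing and a wait-for-catch effect; cleanup turns saber visuals instantly on.
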(event Custom|TriggerThrowFromAI
  (params (AITarget :type Object%20Reference))
  (bind (_owneractor _avataractor) (Utilities|Struct|BreakGameplayAbilityActorInfo (Ability|GetActorInfo)))
  ;[0.9167 0.1205 0.1935] Needs a Custom Event from the AI Side when throwing the Weapon. If this changes again, please advise AI Team.
  (Variables|Default|SetEventHomingTarget AITarget)
  ;[1 1 1] EVENT -> Activate Ability
  (bind _asfort_player_pawn (Utilities|Casting|CastToFortPlayerPawn _avataractor)
    (:then
      (Variables|ReferencedActors|SetPlayerPawn _asfort_player_pawn)
      (ParentDefaults|PrethrowSetup) ; Ability is instanced per actor, so we need to reset variables\nevery time the ability is activated.
      (CallFunction|ThrowProjectile))
    (:CastFailed
      (Ability|EndAbility))))

(event Ability|EventActivateAbilityFromEvent
  (params (EventData :type Gameplay%20Event%20Data%20Structure%20%28by%20ref%29))
  (bind (_eventtag _instigator _target _optionalobject) (Utilities|Struct|BreakGameplayEventData EventData))
  (bind (_owneractor _avataractor) (Utilities|Struct|BreakGameplayAbilityActorInfo (Ability|GetActorInfo)))
  ;[1 1 1] EVENT -> Activate Ability
  (|CallParentFunction EventData)
  (Variables|Default|SetEventHomingTarget _optionalobject)
  (bind _asfort_player_pawn (Utilities|Casting|CastToFortPlayerPawn _avataractor)
    (:then
      (Variables|ReferencedActors|SetPlayerPawn _asfort_player_pawn)
      (ParentDefaults|PrethrowSetup) ; Ability is instanced per actor, so we need to reset variables\nevery time the ability is activated.
      (CallFunction|ThrowProjectile))
    (:CastFailed
      (Ability|EndAbility))))

(event Custom|ThrowProjectile
  ;[1 1 1] Throw Projectile
  (Ability|SetShouldBlockOtherAbilities true)
  (ParentDefaults|PostThrowCleanup)
  ;[1 1 1] Projectile is spawned when the montage reaches ability notify
  (Ability|Tasks|PlayMontageandWaitforTrigger "None" (Variables|Animation|GetThrowMontage) 1.2
    (:then
      (Ability|ExecuteGameplayCueOnOwner (Variables|Defaults|GetGC_OnThrow)) ; GC to play on the player when the projectile is thrown
      (Camera|PushCameraMode (Variables|Default|GetThrowCameraMode)))
    (:Triggered
      (CallFunction|AttemptSpawnThrownProjectile)
      ;;[0 0.1392 1] Allow the player to swap weapons before the throw animation is complete.
      (Utilities|FlowControl|Delay (ScalableFloat|GetValueatLevel (Variables|Default|GetSF_AbilityBlockDuration)))
      (Ability|SetShouldBlockOtherAbilities false))
    (:Cancelled
      _)
    (:Completed
      _))
  (Ability|SetShouldBlockOtherAbilities false)
  (bind _canfinish (CallFunction|HandleShouldFinishAbility))
  (if _canfinish
    (Ability|AbilityCompleted)))

(event Custom|AttemptSpawnThrownProjectile
  ;[1 1 1] Attempt to spawn projectile (only on server)
  (bind _asb_athena_galileo_lobster_realm (Utilities|Casting|CastToB_Athena_Galileo_Lobster_Realm (Variables|/Script/FortniteGame.FortPawn|GetCurrentWeapon (Variables|ReferencedActors|GetPlayerPawn)))
    (:then
      (Variables|/Galileo/Items/Lobster/Actors/B_Athena_Galileo_Lobster_Parent.B_Athena_Galileo_Lobster_Parent_C|SetInstantOn true _asb_athena_galileo_lobster_realm))
    (:CastFailed
      _))
  (if (Networking|IsServer)
    ;[1 1 1] Commit the ability here to prevent exploits relating to spawning multiple projectiles
    (bind _returnvalue (Ability|CommitAbility))
    (if _returnvalue
      ;[1 1 1] Spawn Projectile!
      (Variables|Defaults|SetExplosionEffectContainerSpec (Ability|MakeGameplayEffectContainerSpec (Variables|Defaults|GetEC_DefaultExplosion)))
      (bind (_location _rotation) (Defaults|CalculatePrjSpawnLocationandRotation))
      (bind _spawnedprojectile (ParentDefaults|SpawnThrownProjectile (Variables|ReferencedActors|GetProj_ThrownProjectile) _location _rotation (Variables|Defaults|GetExplosionEffectContainerSpec) (ScalableFloat|GetValueatLevel (Variables|Defaults|GetProjectileSpeed))))
      (bind _output_get (Variables|ReferencedActors|SetProjectileRef _spawnedprojectile))
      ;[1 1 1] Set the homing target based on the target passed in from the GameplayEvent
      (bind _asb_prj_realm_thrown_lobster (Utilities|Casting|CastToB_Prj_Realm_ThrownLobster _output_get)
        (:then
          (bind _eventhomingtarget (Variables|Default|GetEventHomingTarget)
            (:then
              (bind _ascharacter (Utilities|Casting|CastToCharacter _eventhomingtarget)
                (:then
                  (Variables|/Galileo/Items/Lobster/Abilities/Realm/B_Prj_Realm_ThrownLobster.B_Prj_Realm_ThrownLobster_C|SetHomingTarget _ascharacter _asb_prj_realm_thrown_lobster))
                (:CastFailed
                  _))
              ;[1 1 1] Send Event to track projectile. Ability will end when throw montage is complete
              (Ability|ApplyGameplayEffectToOwner (Variables|Default|GetGE%20Wait%20Catch)))
            (:else
              _)))
        (:CastFailed
          _))
      (else
        (Ability|EndAbility)))))

(event Ability|EventCommitExecute
  ;[1 1 1] EVENT -> Commit Execute
  (|CallParentFunction)
  (Ability|SendGameplayEvent "(TagName=\"Ability.Realm.LSThrow.Commit\")"))

(event Ability|EventOnEndAbility
  (params (bWasCancelled :type Boolean))
  ;[1 1 1] EVENT -> End Ability
  (|CallParentFunction bWasCancelled)
  (ParentDefaults|PostThrowCleanup))

(fn CalculatePrjSpawnLocationAndRotation
  (bind (_outlocation _outrotation) (Actor|GetActorEyesViewPoint (ParentDefaults|GetOwningPlayer)))
  (bind _verbosedebug (Variables|Default|GetVerboseDebug))
  (bind _outloc (LocalVariables|GetOutLoc))
  (Utilities|FlowControl|Sequence
    (:then_0
      (LocalVariables|SetOutLoc (+ (Transformation|GetActorLocation (Variables|ReferencedActors|GetPlayerPawn)) (Math|Vector|vector+vector (* (Math|Vector|GetForwardVector (Pawn|GetControlRotation (Variables|ReferencedActors|GetPlayerPawn) :!class /Script/Engine.Pawn)) (Variables|SpawnedProjOffsets|GetSpawnPrjForwardOffset)) (* (Math|Vector|GetRightVector (Pawn|GetControlRotation (Variables|ReferencedActors|GetPlayerPawn) :!class /Script/Engine.Pawn)) (Variables|SpawnedProjOffsets|GetSpawnRightOffset)) (Math|Vector|SelectVector (* (Math|Vector|GetUpVector (Pawn|GetControlRotation (Variables|ReferencedActors|GetPlayerPawn) :!class /Script/Engine.Pawn)) (LocalVariables|GetSPawnUpOffsetRiding)) (select (Variables|/Script/Engine.Character|GetbIsCrouched (Variables|ReferencedActors|GetPlayerPawn)) (* (Math|Vector|GetUpVector (Pawn|GetControlRotation (Variables|ReferencedActors|GetPlayerPawn) :!class /Script/Engine.Pawn)) (Variables|SpawnedProjOffsets|GetSpawnUpOffsetCrouch)) (* (Math|Vector|GetUpVector (Pawn|GetControlRotation (Variables|ReferencedActors|GetPlayerPawn) :!class /Script/Engine.Pawn)) (Variables|SpawnedProjOffsets|GetSpawnUpOffset))) (GameplayTags|HasMatchingGameplayTag (Variables|ReferencedActors|GetPlayerPawn) "(TagName=\"Granted.Riding.Player.IsRiding\")")))))
      (LocalVariables|SetOutRot _outrotation)
      (bind _playerpawn (Variables|ReferencedActors|GetPlayerPawn)
        (:then
          ;[0.573 1 0.8961] Line trace from the actor location to the projectile spawn location to check for any "missed hits". If something is hit, adjust the projectile location so that it is before that hit.
          ; This is necessary since the projectile spawn offset is far enough from the player's location that it can miss hitting something if the player is right up against it.
          (bind (_outhit _returnvalue) (Collision|LineTraceByProfile (Transformation|GetActorLocation _playerpawn) _outloc "FortProjectileThrownOverlapPawns" false (Utilities|Array|MakeArray _playerpawn (Variables|/Script/FortniteGame.FortPawn|GetCurrentWeapon _playerpawn) (Actor|GetAttachParentActor _playerpawn)) "ForDuration" true "(R=1.000000,G=0.000000,B=0.000000,A=1.000000)" "(R=0.000000,G=1.000000,B=0.000000,A=1.000000)" 1e+01)) ; Do NOT worry about pawns  here. The projectile has a much larger pawn specific collision,\nso this issue does not occur when throwing at a close pawn. 
          (if _returnvalue
            (LocalVariables|SetOutLoc (+ (* (Math|Vector|Normalize (- (Transformation|GetActorLocation _playerpawn) _outloc)) (LocalVariables|GetAdjustmentHitOffset)) (.Location (Collision|BreakHitResult _outhit))))
            (Development|PrintString (Utilities|String|Append "Start location trace - Blocking hit, adjusting start location - " (Utilities|GetDisplayName (.HitActor (Collision|BreakHitResult _outhit)))))
            (else
              (Utilities|IsValid (.HitActor (Collision|BreakHitResult _outhit))
                (:Is%20Valid
                  (Development|PrintString (Utilities|String|Append "Start location trace - non blocking hit - " (Utilities|GetDisplayName (.HitActor (Collision|BreakHitResult _outhit)))) _verbosedebug _verbosedebug))
                (:Is%20Not%20Valid
                  (Development|PrintString "Start location trace - No hit" _verbosedebug _verbosedebug))))))
        (:else)))
    (:then_1
      ;[1 0.6073 0.3726] Return OutLoc and OutRot
      (return (LocalVariables|GetOutLoc) (LocalVariables|GetOutRot)))))

(fn GetOwningPlayer
  (bind (_owneractor _avataractor) (Utilities|Struct|BreakGameplayAbilityActorInfo (Ability|GetActorInfo)))
  (Utilities|IsValid (Variables|ReferencedActors|GetPlayerPawn)
    (:Is%20Valid
      (return (Variables|ReferencedActors|GetPlayerPawn)))
    (:Is%20Not%20Valid
      (bind _asfort_player_pawn (Utilities|Casting|CastToFortPlayerPawn _avataractor)
        (:then
          (return _asfort_player_pawn))
        (:CastFailed
          (return 0))))))

(fn GetProjectileTrajectoryActor
  (return 0))

(fn GetProjectileTrajectoryPoints
  (return 0 0 0))

(fn HandleShouldFinishAbility
  ;[1 0 0 0.5] Used in AI Versions of this Ability. Please leave as is.
  (return true))

(fn OnProjectileSetup)

(fn PostThrowCleanup
  (Variables|AbilityState|SetInThrowWindup false)
  (Ability|RemoveGameplayEffectFromOwnerWithHandle (Variables|Defaults|GetHandle_GE_InWindup))
  ;[1 1 1] Makes the lightsaber visuals instantly on upon catching bypassing the "power on" animation
  (bind _asb_athena_galileo_lobster_realm (Utilities|Casting|CastToB_Athena_Galileo_Lobster_Realm (Variables|/Script/FortniteGame.FortPawn|GetCurrentWeapon (Variables|ReferencedActors|GetPlayerPawn))))
  (Variables|/Galileo/Items/Lobster/Actors/B_Athena_Galileo_Lobster_Parent.B_Athena_Galileo_Lobster_Parent_C|SetInstantOn true _asb_athena_galileo_lobster_realm))

(fn PrethrowSetup
  ;[1 1 1] We allow a player to cancel out of the windup for most thrown projectiles
  (Ability|SetShouldBlockOtherAbilities false)
  (Ability|SetCanbeCanceled true)
  ;[1 1 1] The input is being held down and the actor is winding up to throw the projectile
  (Variables|AbilityState|SetAbilityKeyPressed true)
  (Variables|AbilityState|SetInThrowWindup true)
  ;[1 1 1] Add a GE that applies looping GC while in the windup
  (bind _returnvalue (Ability|ApplyGameplayEffectToOwner (Variables|Defaults|GetGE_InWindup)))
  (Variables|Defaults|SetHandle_GE_InWindup _returnvalue))

(fn ShouldOnlyShowTrajectoryOnUse
  (return false))

(fn SpawnThrownProjectile
  (params (ProjectileClass :type Fort%20Projectile%20Base%20Class%20Reference) (ProjectileSpawnLocation :type Vector%20%28by%20ref%29) (ProjectileSpawnDirection :type Rotator%20%28by%20ref%29) (ExplosionEffectContainerSpec :type Fort%20Gameplay%20Effect%20Container%20Spec%20Structure) (ProjectileInitialSpeed :type Float%20%28double-precision%29) (ProjectileGravityScale :type Float%20%28double-precision%29))
  (bind _returnvalue (Projectile|SpawnProjectile :ProjectileClass ProjectileClass :RequestedBy (ParentDefaults|GetOwningPlayer) :SpawnLocation ProjectileSpawnLocation :SpawnRotation ProjectileSpawnDirection :SpawnDirection ProjectileSpawnDirection :EffectContainerSpecToApplyOnExplode ExplosionEffectContainerSpec :bUseDefaultPhysics false :InitialSpeed ProjectileInitialSpeed :GravityScale ProjectileGravityScale :ChargePercent 0.0))
  (bind _output_get (Variables|ReferencedActors|SetSpawnedProjectile _returnvalue))
  (return _output_get))

(fn SpawnTrajectoryIndicator
  (params (bSpawnOnEquip :type Boolean))
  (return 0))
